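# Niagara idle effect for the tower pickup held-object gadget, layering glows, smoke, flares and a wings mesh.
# User.bIsHeld switches spawning from large alternating up/down light spikes to smaller random spikes, blue smoke and solar flares; a rainbow corona always spawns.
# A persistent wings particle trails the owner within a clamped offset, blending between slow spin and alignment to User.ObeliskPosition through a time-smoothed User.bIsFacing.
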NiagaraSystem /PrimalGameplay/Items/HeldObjects/Assets/Tower_Pickup/Effects/NS_Gadget_TowerPickup_Idle
dependencies {
 usedDataInterfaces [
  &NiagaraDataInterfaceCurve
  &NiagaraDataInterfaceVectorCurve
  /Script/Niagara.NiagaraDataInterfaceSimpleCounter
 ]
 usedDynamicInputs []
 usedModules [&SystemState]
}
emitters {
 BigLightSpikes {
  emitterUpdateScript [
   &EmitterState()
   &SpawnRate(
    Spawn Probability=&MakeCustomFloatFromBool(Bool=$User.bIsHeld, True Float=0, False Float=1)
    SpawnRate=4
    Use Spawn Probability=true
   )
  ]
  particleSpawnScript [
   /SetVariables(Particles.IsTop:NiagaraBool=&SetBoolByIntComparison(
     A=&ModuloInt(
      A=/Niagara/DynamicInputs/DataInterface/GenerateIntFromCounter(SimpleCounter:NiagaraDataInterfaceSimpleCounter=DataInterface() {propertyValues {GpuSyncMode=None, InitialValue=0}})
      B=2
     )
     B=0
     Comparison Type=::NewEnumerator4
    )) {isSetParameters=true}
   &InitializeParticle(
    AdjustAlpha=true
    AdjustHue=true
    AdjustSaturation=true
    AdjustValue=true
    Alpha Scale Range:Vector2f=<0.15 0.3>
    Color=<rgba 0 0.3681 1 1>
    Color Mode=::NewEnumerator3
    Hue Shift Range:Vector2f=<-0.15 0.15>
    Lifetime Min=1.5
    Lifetime Mode=::NewEnumerator1
    Mesh Scale Mode=::NewEnumerator3
    Mesh Uniform Scale=&MakeCustomFloatFromBool(
     Bool=$Particles.IsTop
     True Float=&RandomRangeFloat(Minimum=1.5, Maximum=2.5)
     False Float=&RandomRangeFloat(Minimum=0.6, Maximum=1)
    )
    Saturation Range:Vector2f=<0.3 0.5>
    UsePositionOffset=false
    Value Range:Vector2f=<3 4>
    Write Material Random=true
   )
   /Niagara/Modules/Update/Orientation/V2/OrientMeshToVector(
    Look At Direction:Vector3f=&TransformVector(
     Vector=&MakeCustomVectorFromBool(Bool=$Particles.IsTop, True Vector=<0 0 1>, False Vector=<0 0 -1>)
     Source Space=::World
     Destination Space=::Simulation
    )
    Override Quat=false
    UseOptionalSideAxis=false
    UseOptionalUpAxis=false
   )
  ]
  particleUpdateScript [
   &ParticleState()
   &DynamicMaterialParameters(
    Index 0 Param 1=&FloatFromCurve(
     FloatCurve=DataInterface() {propertyValues {Curve {keys [
         {interpMode=RCIM_Cubic, tangentMode=RCTM_User, time=0, value=1}
         {arriveTangent=-1.125, interpMode=RCIM_Cubic, leaveTangent=-1.125, tangentMode=RCTM_User, time=1, value=0.5}
        ]}}}
     Scale Curve=&RandomRangeFloat(Minimum=1, Maximum=8)
    )
    Index 0 Param 2=&RandomRangeFloat(Minimum=0.03, Maximum=0.05)
    Index 0 Param 3=&RandomRangeFloat(Minimum=0, Maximum=1)
    Index 0 Param 4=1
    Module.Param0WriteEnabled=true
    Module.Param1WriteEnabled=true
    Module.Param2WriteEnabled=true
    Module.Param3WriteEnabled=true
   )
   &ScaleColor(
    Scale Alpha=&FloatFromCurve(FloatCurve=DataInterface() {propertyValues {Curve {keys [
         <at 0 0 RCIM_Cubic>
         {arriveTangent=-0.3126, interpMode=RCIM_Cubic, leaveTangent=-0.3126, tangentMode=RCTM_User, time=0.2, value=1}
         {arriveTangent=-1.25, interpMode=RCIM_Cubic, leaveTangent=-1.25, tangentMode=RCTM_User, time=0.7442, value=0.7365}
         <at 1 0 RCIM_Cubic>
        ]}}})
    Scale RGB=&VectorFromCurve(VectorCurve=DataInterface() {propertyValues {XCurve {keys [<at 0 1>,<at 1 0>]}, YCurve {keys [<at 0 1>,<at 1 0>]}, ZCurve {keys [<at 0 1>,<at 1 2>]}}})
    ScaleA=true
    ScaleRGB=true
   )
   &ScaleMeshSize(Scale Factor=&VectorFromFloat(Value=&FloatFromCurve(FloatCurve=DataInterface() {propertyValues {Curve {keys [
          {arriveTangent=6.7832, interpMode=RCIM_Cubic, leaveTangent=6.7832, tangentMode=RCTM_User, time=0, value=0}
          {arriveTangent=0.0353, interpMode=RCIM_Cubic, leaveTangent=0.0353, tangentMode=RCTM_User, time=0.25, value=1}
          <at 1 0.75>
         ]}}})))
   /SetVariables() {isSetParameters=true}
  ]
  rendererClasses [&NiagaraMeshRendererProperties]
  renderers [{rendererClass=&NiagaraMeshRendererProperties, rendererIndex=0}]
 }
 BlueSmoke {
  emitterUpdateScript [
   &EmitterState()
   &SpawnRate(
    Spawn Probability=&MakeCustomFloatFromBool(Bool=$User.bIsHeld, True Float=1, False Float=0)
    SpawnRate=6
    Use Spawn Probability=true
   )
  ]
  particleSpawnScript [
   /Niagara/Modules/Spawn/Initialization/InitializeParticle(
    Color=&RandomRangeLinearColor(Minimum=<rgba 0.735 0.7 1 0.1>, Maximum=<rgba 0.5 0.7083 1 0.1>, Color Channel Mode=::NewEnumerator2)
    Lifetime=&UniformRangedFloat(Minimum=1, Maximum=2)
    Mesh Scale=<3 3 3>
    Module.Write Color=true
    Module.Write Lifetime=true
    Module.Write Mass=false
    Module.Write Material Random=true
    Module.Write Position=false
    Module.Write Scale=true
    Module.Write SpriteRotation=true
    Module.Write SpriteSize=true
    Position=&SimulationPosition()
    Sprite Rotation=&RandomRangeFloat(Minimum=-360, Maximum=360)
    Sprite Size=&Vector2DFromFloat(Value=&RandomRangeFloat(Minimum=40, Maximum=60))
   )
   &SphereLocation(
    MaskSpawnGroup=false
    Non Uniform Scale=<1 1 0.5>
    Offset=$System.BurstOffset
    Offset Coordinate Space:ENiagaraCoordinateSpace=::World
    Override Seed=false
    Sphere Coordinate Space=::World
    Sphere Radius=5
    Surface Only=false
    UseNonUniformScale=true
    UseOffset=true
    UseSphereOrientation=false
   )
  ]
  particleUpdateScript [
   &ParticleState()
   /SetVariables(Transient.DistFromBase:NiagaraFloat=&ClampFloat(Float=&Add_Float(
      A=0
      B=&Multiply_Float(
       A=0.015
       B=&VectorLength(Vector=&Subtract_Vector(
         A=Linked() {linkedVariable {name=Particles.Position, type=NiagaraPosition}}
         B=Linked() {linkedVariable {name=Particles.Initial.Position, type=NiagaraPosition}}
        ))
      )
     ))) {isSetParameters=true}
   &SpriteRotationRate(Rotation Rate=&Multiply_Float(
     A=&RandomRangeFloat(Minimum=-3, Maximum=3)
     B=&FloatFromCurve(FloatCurve=DataInterface() {propertyValues {Curve {keys [<at -0.0046 1.0113 RCIM_Cubic>,<at 1.0011 0.09 RCIM_Cubic>]}}})
    ))
   /Niagara/Modules/Update/SubUV/SubUVAnimation(
    Number Of Frames=64
    SubUV Lookup Index=&Add_Float(
     A=&RandomRangeFloat(Minimum=0, Maximum=0.5)
     B=&Lerp_Float(A=0, B=&RandomRangeFloat(Minimum=0.25, Maximum=0.5), Alpha=$Particles.NormalizedAge)
    )
   )
   &Drag(Drag=10, Use Linear Drag=true, Use Rotational Drag=false)
   &GravityForce(Gravity=&RandomRangeVector(Minimum=<0 0 -300>, Maximum=<0 0 600>))
   &SolveForcesAndVelocity(Clamp Velocity=false, Limit Acceleration=false)
   &ScaleSpriteSize(Scale Factor=&Vector2DFromFloat(Value=&FloatFromCurve(
      FloatCurve=DataInterface() {propertyValues {Curve {keys [
          {arriveTangent=2.7917, interpMode=RCIM_Cubic, leaveTangent=2.7735, tangentMode=RCTM_User, time=0, value=0.5}
          {arriveTangent=1.2109, interpMode=RCIM_Cubic, leaveTangent=4.0244, tangentMode=RCTM_User, time=1, value=2}
         ]}}}
      Scale Curve=&Lerp_Float(A=1, B=0.25, Alpha=$Transient.DistFromBase)
     )))
   &DynamicMaterialParameters(
    Index 0 Param 1=&RandomRangeFloat(Minimum=0.25, Maximum=2)
    Index 0 Param 2=0.2
    Index 0 Param 3=0.1
    Index 0 Param 4=1
    Module.Param0WriteEnabled=true
    Module.Param1WriteEnabled=true
    Module.Param2WriteEnabled=true
    Module.Param3WriteEnabled=true
   )
   &ScaleColor(
    Scale Alpha=&FloatFromCurve(FloatCurve=DataInterface() {propertyValues {Curve {keys [
         <at 0 0 RCIM_Cubic>
         {arriveTangent=-0.0914, interpMode=RCIM_Cubic, leaveTangent=-0.0914, tangentMode=RCTM_User, time=0.125, value=1}
         {arriveTangent=-1.1429, interpMode=RCIM_Cubic, leaveTangent=-1.1429, tangentMode=RCTM_User, time=0.5734, value=0.764}
         {arriveTangent=-1.1414, interpMode=RCIM_Cubic, leaveTangent=-1.1414, tangentMode=RCTM_User, time=1, value=0}
        ]}}})
    Scale RGB=&VectorFromCurve(VectorCurve=DataInterface() {propertyValues {
       XCurve {keys [{arriveTangent=-1.2088, interpMode=RCIM_Cubic, leaveTangent=-1.2088, tangentMode=RCTM_User, time=0, value=1}, <at 1 0.1>]}
       YCurve {keys [{arriveTangent=-1.1551, interpMode=RCIM_Cubic, leaveTangent=-1.1551, tangentMode=RCTM_User, time=0, value=1}, <at 1 0.1>]}
       ZCurve {keys [{arriveTangent=-0.3552, interpMode=RCIM_Cubic, leaveTangent=-0.3552, tangentMode=RCTM_User, time=0, value=1}, <at 1 2>]}
      }})
    ScaleA=true
    ScaleRGB=true
   )
  ]
  rendererClasses [&NiagaraSpriteRendererProperties]
  renderers [{rendererClass=&NiagaraSpriteRendererProperties}]
 }
 LightSpikes {
  emitterUpdateScript [
   &EmitterState()
   &SpawnRate(
    Spawn Probability=&MakeCustomFloatFromBool(Bool=$User.bIsHeld, True Float=0.5, False Float=0)
    SpawnRate=8
    Use Spawn Probability=true
   )
  ]
  particleSpawnScript [
   /Niagara/Modules/Spawn/Initialization/V2/InitializeParticle(
    AdjustAlpha=true
    AdjustHue=true
    AdjustSaturation=true
    AdjustValue=true
    Alpha Scale Range:Vector2f=<0.15 0.3>
    Color=<rgba 0.4436 0 1 1>
    Color Mode=::NewEnumerator3
    Hue Shift Range:Vector2f=<-0.35 0.35>
    Lifetime Mode=::NewEnumerator1
    Mesh Scale Mode=::NewEnumerator3
    Mesh Uniform Scale=&Lerp_Float(A=0.4, B=1, Alpha=&Power(a=&RandomRangeFloat(Minimum=0, Maximum=1), b=3))
    Saturation Range:Vector2f=<0.3 0.5>
    UsePositionOffset=false
    Value Range:Vector2f=<1 4>
    Write Material Random=true
   )
   /SetVariables(Particles.MeshOrientation=/Niagara/DynamicInputs/Rotation/RandomQuaternion()) {isSetParameters=true}
  ]
  particleUpdateScript [
   &ParticleState()
   &DynamicMaterialParameters(
    Index 0 Param 1=&RandomRangeFloat(Minimum=1, Maximum=5)
    Index 0 Param 2=&RandomRangeFloat(Minimum=0.03, Maximum=0.05)
    Index 0 Param 3=&RandomRangeFloat(Minimum=0, Maximum=1)
    Index 0 Param 4=1
    Module.Param0WriteEnabled=true
    Module.Param1WriteEnabled=true
    Module.Param2WriteEnabled=true
    Module.Param3WriteEnabled=true
   )
   &ScaleColor(
    Scale Alpha=&FloatFromCurve(FloatCurve=DataInterface() {propertyValues {Curve {keys [
         <at 0 0 RCIM_Cubic>
         {arriveTangent=-1.9443, interpMode=RCIM_Cubic, leaveTangent=-1.9443, tangentMode=RCTM_User, time=0.2, value=1}
         <at 1 0 RCIM_Cubic>
        ]}}})
    Scale RGB=&VectorFromCurve(VectorCurve=DataInterface() {propertyValues {XCurve {keys [<at 0 1>,<at 1 0>]}, YCurve {keys [<at 0 1>,<at 1 0>]}, ZCurve {keys [<at 0 1>,<at 1 2>]}}})
    ScaleA=true
    ScaleRGB=true
   )
   &ScaleMeshSize(Scale Factor=&VectorFromFloat(Value=&FloatFromCurve(FloatCurve=DataInterface() {propertyValues {Curve {keys [
          {arriveTangent=6.7832, interpMode=RCIM_Cubic, leaveTangent=6.7832, tangentMode=RCTM_User, time=0, value=0}
          {arriveTangent=0.0353, interpMode=RCIM_Cubic, leaveTangent=0.0353, tangentMode=RCTM_User, time=0.25, value=1}
          <at 1 0.75>
         ]}}})))
  ]
  rendererClasses [&NiagaraMeshRendererProperties]
  renderers [{rendererClass=&NiagaraMeshRendererProperties, rendererIndex=0}]
 }
 RainbowCorona {
  emitterUpdateScript [
   &EmitterState()
   &SpawnRate(SpawnRate=3, Use Spawn Probability=false)
  ]
  particleSpawnScript [&InitializeParticle(
    AdjustAlpha=true
    AdjustHue=true
    AdjustSaturation=true
    AdjustValue=false
    Alpha Scale Range:Vector2f=<0.075 0.2>
    Color=<rgba 1 0.63 0.63 1>
    Color Mode=::NewEnumerator3
    Hue Shift Range:Vector2f=<-1 1>
    Lifetime=1.5
    Saturation Range:Vector2f=<0.75 1>
    Sprite Rotation Mode=::NewEnumerator2
    Sprite Size Mode=::NewEnumerator1
    Uniform Sprite Size Max=125
    Uniform Sprite Size Min=100
    UsePositionOffset=false
    Write Material Random=true
   )]
  particleUpdateScript [
   &ParticleState()
   &ScaleSpriteSize(Scale Factor=&Vector2DFromFloat(Value=&FloatFromCurve(FloatCurve=DataInterface() {propertyValues {Curve {keys [
          {arriveTangent=3.8293, interpMode=RCIM_Cubic, leaveTangent=3.8293, tangentMode=RCTM_User, time=0, value=0.5}
          {arriveTangent=0.4994, interpMode=RCIM_Cubic, leaveTangent=0.4994, tangentMode=RCTM_User, time=0.0999, value=0.8637}
          {arriveTangent=0.0432, interpMode=RCIM_Cubic, leaveTangent=0.0432, tangentMode=RCTM_User, time=1.0011, value=1}
         ]}}})))
   &DynamicMaterialParameters(
    Index 0 Param 1=&RandomRangeFloat(Minimum=0.75, Maximum=1.5)
    Index 0 Param 2=&RandomRangeFloat(Minimum=-0.4, Maximum=0.4)
    Index 0 Param 3=1
    Index 0 Param 4=1
    Index 1 Param 1=&RandomRangeFloat(Minimum=2.5, Maximum=5)
    Index 1 Param 2=&FloatFromCurve(
     FloatCurve=DataInterface() {propertyValues {Curve {keys [
         {arriveTangent=10.7118, interpMode=RCIM_Cubic, leaveTangent=10.7118, tangentMode=RCTM_User, time=0, value=0}
         {interpMode=RCIM_Cubic, tangentMode=RCTM_User, time=0.2, value=1}
         <at 1 0>
        ]}}}
     Scale Curve=&RandomRangeFloat(Minimum=0.1, Maximum=0.7)
    )
    Index 1 Param 3=2
    Module.Param0WriteEnabled=true
    Module.Param0WriteEnabled1=true
    Module.Param1WriteEnabled=true
    Module.Param1WriteEnabled1=true
    Module.Param2WriteEnabled=true
    Module.Param2WriteEnabled1=true
    Module.Param3WriteEnabled=true
    Module.Param3WriteEnabled1=true
    Write Parameter Index 1=true
   )
   &ScaleColor(
    Scale Alpha=&FloatFromCurve(FloatCurve=DataInterface() {propertyValues {Curve {keys [
         <at 0 0 RCIM_Cubic>
         {arriveTangent=-0.3825, interpMode=RCIM_Cubic, leaveTangent=-0.3825, tangentMode=RCTM_User, time=0.3, value=1}
         {arriveTangent=-1.6815, interpMode=RCIM_Cubic, leaveTangent=-1.6815, tangentMode=RCTM_User, time=1.0011, value=0.002}
        ]}}})
    ScaleA=true
    ScaleRGB=false
   )
   &CameraOffset(Camera Offset Amount=5)
   &SpriteRotationRate(Rotation Rate=&RandomRangeFloat(Minimum=-4, Maximum=4))
  ]
  rendererClasses [&NiagaraSpriteRendererProperties]
  renderers [{rendererClass=&NiagaraSpriteRendererProperties}]
 }
 SolarFlares {
  emitterUpdateScript [
   &EmitterState()
   &SpawnRate(
    Spawn Probability=&MakeCustomFloatFromBool(Bool=$User.bIsHeld, True Float=1, False Float=0)
    SpawnRate=2
    Use Spawn Probability=true
   )
  ]
  particleSpawnScript [
   /Niagara/Modules/Spawn/Initialization/InitializeParticle(
    Color=&RandomRangeLinearColor(Minimum=<rgba 0.6061 1.1919 5 0.5>, Maximum=<rgba 1.0455 0.6061 5 0.5>, Color Channel Mode=::NewEnumerator2)
    Lifetime=3.6
    Module.Write Color=true
    Module.Write Lifetime=true
    Module.Write Mass=true
    Module.Write Material Random=true
    Module.Write Position=false
    Module.Write Scale=false
    Module.Write SpriteRotation=true
    Module.Write SpriteSize=true
    Position=&SimulationPosition()
    Sprite Rotation=&RandomRangeFloat(Minimum=-180, Maximum=180)
    Sprite Size=&Vector2DFromFloat(Value=&RandomRangeFloat(Minimum=30, Maximum=60))
   )
   /SetVariables(Particles.RandomFrame=&RandomRangeInt(Minimum=0, Maximum=63)) {isSetParameters=true}
   &SphereLocation(
    MaskSpawnGroup=false
    Override Seed=false
    Sphere Radius=20
    Surface Only=true
    UseNonUniformScale=false
    UseOffset=false
    UseSphereOrientation=false
   )
  ]
  particleUpdateScript [
   &ParticleState()
   &CameraOffset(Camera Offset Amount=5)
   &FN_SubUVAnimation(End Frame=&Add_Integer(A=$Particles.RandomFrame, B=63), Start Frame=$Particles.RandomFrame, SubUV Loop Count=0.75)
   &ScaleSpriteSize(Scale Factor=&Vector2DFromFloat(Value=&FloatFromCurve(
      FloatCurve=DataInterface() {propertyValues {Curve {keys [
          {arriveTangent=2.5076, interpMode=RCIM_Cubic, leaveTangent=2.5076, tangentMode=RCTM_User, time=0.0023, value=0}
          {arriveTangent=0.0897, interpMode=RCIM_Cubic, leaveTangent=0.0897, tangentMode=RCTM_User, time=1, value=1}
         ]}}}
      Scale Curve=1.2
     )))
   &ScaleColor(
    Scale Alpha=&FloatFromCurve(FloatCurve=DataInterface() {propertyValues {Curve {keys [
         <at 0 0 RCIM_Cubic>
         {arriveTangent=5.1321, interpMode=RCIM_Cubic, leaveTangent=0.2, tangentMode=RCTM_Break, time=0.5, value=1}
         {arriveTangent=-1.4896, interpMode=RCIM_Cubic, leaveTangent=-1.4896, tangentMode=RCTM_User, time=1, value=0}
        ]}}})
    Scale RGB=&Lerp_Vector(A=<1 1 1>, B=<0 0 1>, Alpha=$Particles.NormalizedAge)
    ScaleA=true
    ScaleRGB=true
   )
  ]
  rendererClasses [&NiagaraSpriteRendererProperties]
  renderers [{rendererClass=&NiagaraSpriteRendererProperties}]
 }
 WingsStaticMesh {
  emitterUpdateScript [
   &EmitterState()
   &SpawnBurst_Instantaneous(
    Spawn Count=1
    Spawn Probability=&MakeCustomFloatFromBool(
     Bool=&SetBoolByIntComparison(A=$Emitter.LoopCount, B=0, Comparison Type=::NewEnumerator4)
     True Float=1
     False Float=0
    )
    Use Spawn Probability=true
   )
  ]
  particleSpawnScript [
   /Niagara/Modules/Spawn/Initialization/V2/InitializeParticle(
    Color Mode=::NewEnumerator0
    Lifetime=5
    Mesh Scale Mode=::NewEnumerator3
    Sprite Size Mode=::NewEnumerator3
    Uniform Sprite Size=100
    UsePositionOffset=false
    Write Material Random=true
   )
   &InitialMeshOrientation(Use Orientation Vector=true, Use Rotation Vector=false)
   /SetVariables() {isSetParameters=true}
   /SetVariables() {isSetParameters=true}
  ]
  particleUpdateScript [
   /SetVariables(Particles.isFacingSmooth:NiagaraFloat=&ClampFloat(Float=&Add_Float(
      A=$Particles.isFacingSmooth
      B=&Multiply_Float(A=$Engine.DeltaTime, B=&MakeCustomFloatFromBool(Bool=$User.bIsFacing, True Float=0.2, False Float=-1))
     ))) {isSetParameters=true}
   &ParticleState(Kill Particles When Lifetime Has Elapsed=false, Let Infinitely Lived Particles Die When Emitter Deactivates=true)
   /SetVariables(Particles.Position=/Niagara/DynamicInputs/Math/Clamp_Vector(
     Vector:Vector3f=&self:InterpolateVector(
      Initial Value:Vector3f=Linked() {linkedVariable {name=Engine.Owner.Position, type=NiagaraPosition}}
      Vector to Interpolate:Vector3f=Linked() {linkedVariable {name=Engine.Owner.Position, type=NiagaraPosition}}
      Rate Of Change:NiagaraFloat=15
     )
     Min:Vector3f=&Add_Vector(A=Linked() {linkedVariable {name=Engine.Owner.Position, type=NiagaraPosition}}, B=<-8 -8 -8>)
     Max:Vector3f=&Add_Vector(A=Linked() {linkedVariable {name=Engine.Owner.Position, type=NiagaraPosition}}, B=<8 8 8>)
    )) {isSetParameters=true}
   &self:OrientMeshSimple(
    Look At Direction:Vector3f=&Subtract_Vector(A=Linked() {linkedVariable {name=Particles.Position, type=NiagaraPosition}}, B=$User.ObeliskPosition)
    Up Axis Direction:Vector3f=<0 0 1>
   ) {module=OrientMeshSimple}
   &self:RotateMeshSimple(
    RotatingOrientation:Quat4f=/Niagara/DynamicInputs/Vectors/OrientMeshToVector(
     Look At Direction=&Multiply_Vector(
      A=&ReturnMeshOrientationAxis(Mesh Orientation Axis=::XAxis, Mesh Orientation Quat=$Particles.MeshOrientation)
      B=<1 1 0>
     )
     Reference Axis Vector=<1 0 0>
    )
    Rotation Rate=0.25
    Rotation Vector:Vector3f=<0 0 1>
   ) {module=RotateMeshSimple}
   /SetVariables(Particles.MeshOrientation=&LerpQuaternion(
     Quaternion A:Quat4f=$Particles.RotatingOrientation
     Quaternion B:Quat4f=$Particles.FacingOrientation
     Lerp Factor:NiagaraFloat=$Particles.isFacingSmooth
     Enable Spherical Linear Interpolation:NiagaraBool=false
    )) {isSetParameters=true}
   /SetVariables(Particles.MeshOrientation=&LerpQuaternion(
     Quaternion A:Quat4f=$Particles.RotatingOrientation
     Quaternion B:Quat4f=$Particles.FacingOrientation
     Lerp Factor:NiagaraFloat=&self:InterpolateFloat(
      Initial Value:NiagaraFloat=&MakeCustomFloatFromBool(Bool=$User.bIsFacing, True Float=1, False Float=0)
      Target Value:NiagaraFloat=&MakeCustomFloatFromBool(Bool=$User.bIsFacing, True Float=1, False Float=0)
      Rate Of Change:NiagaraFloat=2
     )
     Enable Spherical Linear Interpolation:NiagaraBool=false
    )) {enabled=false, isSetParameters=true}
  ]
  rendererClasses [&NiagaraMeshRendererProperties]
  renderers [{rendererClass=&NiagaraMeshRendererProperties, rendererIndex=0}]
 }
}
summary {
 systemName=NS_Gadget_TowerPickup_Idle
 emitters [
  {emitterName=BlueSmoke, rendererClasses [&NiagaraSpriteRendererProperties]}
  {emitterName=RainbowCorona, rendererClasses [&NiagaraSpriteRendererProperties]}
  {emitterName=LightSpikes, rendererClasses [&NiagaraMeshRendererProperties]}
  {emitterName=BigLightSpikes, rendererClasses [&NiagaraMeshRendererProperties]}
  {emitterName=SolarFlares, rendererClasses [&NiagaraSpriteRendererProperties]}
  {emitterName=WingsStaticMesh, rendererClasses [&NiagaraMeshRendererProperties]}
 ]
 userVariables [
  User.bIsFacing {type=NiagaraBool}
  User.bIsHeld {type=NiagaraBool}
  User.ObeliskPosition {type=Vector3f, description="", defaultValue {struct=Vector3f, value=<50 1.2646 0>}}
  User.Stage {type=NiagaraInt32}
 ]
}
